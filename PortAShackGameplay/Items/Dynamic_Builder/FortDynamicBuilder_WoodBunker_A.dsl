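; Dynamic builder for the Port-A-Bunker (wood variant) that fixes up conflicts after its pieces spawn, server-only.
; Once the playlist is ready, a hotfix-gated binding triggers on spawn completion, killing pieces overlapping nearby indestructible gameplay actors of matching building type.
; It then line-traces tagged pieces to move overlapping grounded players onto them within a distance limit.
; Finally it broadcasts a deploy quest verb, tagged when pieces overlap water.
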(event EventBeginPlay
  ;[0.08266 0.6302 0.1529] BeginPlay: On the server, bind to know when the building actors are all spawned.
  (|CallParentFunction)
  (Networking|SwitchHasAuthority
    (:Authority
      (Class|FortAsyncActionCurrentPlaylistReady|PlaylistDataReadyAsync
        (:OnReady
          (if (ScalableFloat|GetValueasBool (Variables|Default|GetHotfix_AllowDestroySpawnedBuildingOnNonDestructibleBGA))
            (EventDispatchers|BindEventtoBPDynamicBuilderFinishedSpawning (EventDispatchers|CreateEvent :!event HandleBuildingsSpawned))))))))

(event EventEndPlay
  (params (EndPlayReason :type EEndPlayReason%20Enum :default "Destroyed"))
  ;[0.2083 0.1516 0.02235] EndPlay: Cleanup bound delegates.
  (|CallParentFunction EndPlayReason)
  (Networking|SwitchHasAuthority
    (:Authority
      (EventDispatchers|UnbindEventfromBPDynamicBuilderFinishedSpawning (EventDispatchers|CreateEvent :!event HandleBuildingsSpawned)))))

(fn AdjustOverlappingPlayerPawnLocations
  (bind _pawntoteleportlocationmap (LocalVariables|GetPawnToTeleportLocationMap))
  (bind _pawntoteleportlocationmap_2 (LocalVariables|GetPawnToTeleportLocationMap))
  ;[0.04365 0.1485 0.1562] AdjustOverlappingPlayerPawnLocations: Adjust the location of pawns that are clipped by the spawned building actors.
  (Networking|SwitchHasAuthority
    (:Authority
      (if (ScalableFloat|GetValueasBool (Variables|Default|GetHotfix_AdjustOverlappingPlayerPawnLocations))
        (for _array_element (Variables|Default|GetSpawnedBuildingActors)
          (Utilities|IsValid _array_element
            (:Is%20Valid
              ;;[0.1458 0.0367 0.1072] Run this logic only for clipping caused by tagged building actors.
              (bind _asbuilding_static_mesh (Utilities|Casting|CastToBuildingSMActor _array_element))
              (if (Actor|ActorHasTag _asbuilding_static_mesh "Portashack_AdjustOverlappingPlayer")
                (bind _staticmeshcomponent (Variables|/Script/FortniteGame.BuildingSMActor|GetStaticMeshComponent _asbuilding_static_mesh)
                  (:then
                    ;;[0.1109 0.1562 0.03126] Find overlapping player pawns.
                    (bind _outactors (Collision|ComponentOverlapActors _staticmeshcomponent (Transformation|GetWorldTransform _staticmeshcomponent) (Utilities|Array|MakeArray "ObjectTypeQuery3") "/Script/FortniteGame.FortPlayerPawnAthena"))
                    (for _array_element_2 _outactors
                      (bind _asfort_player_pawn_athena (Utilities|Casting|CastToFortPlayerPawnAthena _array_element_2))
                      (Fortnite|FortShippingLog self (Utilities|String|Append "AdjustOverlappingPlayerPawnLocations: Overlapping player: " (Fortnite|GetPlayerNameSafe _asfort_player_pawn_athena) " ; BuildingActor: " (Utilities|GetDisplayName _asbuilding_static_mesh)))
                      ;;[0.1562 0.02243 0.114] Trace against the building static mesh component to find a spot on it that the player can be moved to.
                      (bind (_hitlocation _hitnormal _bonename _outhit _returnvalue) (Collision|LineTraceComponent _staticmeshcomponent (+ (Transformation|GetActorLocation _asfort_player_pawn_athena) "0,0,500.000000") (- (Transformation|GetActorLocation _asfort_player_pawn_athena) "0.000000,0,500.000000") true true true))
                      (if _returnvalue
                        (bind _capsulecomponent (Variables|/Script/Engine.Character|GetCapsuleComponent _asfort_player_pawn_athena)
                          (:then
                            ;;[0.1042 0.08307 0.005116] Track the lowest location to teleport a player.
                            (if _returnvalue
                              (if (.ReturnValue (Utilities|Map|Find _pawntoteleportlocationmap _asfort_player_pawn_athena))
                                (if (> (.z (Utilities|Map|Find _pawntoteleportlocationmap _asfort_player_pawn_athena)) (.z (Utilities|Operators|Add _hitlocation (Math|Vector|MakeVector 0.0 0.0 (Components|Capsule|GetScaledCapsuleHalfHeight _capsulecomponent)))))
                                  _)
                                (else
                                  _))
                              (Utilities|Map|Add (LocalVariables|GetPawnToTeleportLocationMap) _asfort_player_pawn_athena (Utilities|Operators|Add _hitlocation (Math|Vector|MakeVector 0.0 0.0 (Components|Capsule|GetScaledCapsuleHalfHeight _capsulecomponent))))))
                          (:else)))))
                  (:else))))))
        (bind _keys (Utilities|Map|Keys _pawntoteleportlocationmap_2))
        (for _array_element_3 _keys
          (bind _charactermovement (Variables|/Script/Engine.Character|GetCharacterMovement _array_element_3)
            (:then
              (if (or (Pawn|Components|CharacterMovement|IsWalking _charactermovement) (FortMovement|IsRidingonActor _array_element_3))
                (if (< (Math|Vector|DistanceSquared(Vector) (Utilities|Map|Find _pawntoteleportlocationmap_2 _array_element_3) (Transformation|GetActorLocation _array_element_3)) (ScalableFloat|GetValueatLevel (Variables|Default|GetHotfix_MaxAdjustOverlappingPlayerPawnDistanceSq)))
                  (Transformation|SetActorLocation _array_element_3 (Utilities|Map|Find _pawntoteleportlocationmap_2 _array_element_3) false true)
                  (Pawn|Components|CharacterMovement|SetMovementMode _charactermovement "MOVE_Falling" 0) ; Setting to falling refinds the floor on the movement component.
                  (Fortnite|FortShippingLog self (Utilities|String|Append "AdjustOverlappingPlayerPawnLocations: Adjusting player location:  " (Fortnite|GetPlayerNameSafe _array_element_3)))
                  (else
                    (Fortnite|FortShippingLog self (Utilities|String|Append "AdjustOverlappingPlayerPawnLocations: Skipping adjust location since adjust distance is too large for player:   " (Fortnite|GetPlayerNameSafe _array_element_3) "; said distance = " (Utilities|String|ToString(Float) (Math|Float|Sqrt (Math|Vector|DistanceSquared(Vector) (Utilities|Map|Find _pawntoteleportlocationmap_2 _array_element_3) (Transformation|GetActorLocation _array_element_3))))))))))
            (:else)))))))

(fn HandleBuildingsSpawned
  (params (Buildings :type Array%20of%20Building%20Actor%20Object%20References))
  ;[0.1562 0.0221 0.1437] HandleBuildingsSpawned: On the server, out of all the buildings that are spawned, destroy the ones that overlap an indestructible building gameplay actor and are of the same building type.
  ; This prevents building issues around building actors such as vault doors.
  ; In addition, adjust the location of any player that may be overlapped/block by the spawned building actors.
  (Networking|SwitchHasAuthority
    (:Authority
      (Utilities|FlowControl|Sequence
        (:then_0
          (Variables|Default|SetSpawnedBuildingActors Buildings)
          (for _array_element Buildings
            (Utilities|IsValid _array_element
              (:Is%20Valid
                (bind _asbuilding_static_mesh (Utilities|Casting|CastToBuildingSMActor _array_element))
                (bind _staticmeshcomponent (Variables|/Script/FortniteGame.BuildingSMActor|GetStaticMeshComponent _asbuilding_static_mesh)
                  (:then
                    (bind _outactors (Collision|ComponentOverlapActors _staticmeshcomponent (Transformation|GetWorldTransform _staticmeshcomponent) (Utilities|Array|MakeArray "ObjectTypeQuery1" "ObjectTypeQuery2") "/Script/FortniteGame.BuildingGameplayActor" Buildings))
                    (for _array_element_2 _outactors
                      (bind _asbuilding_actor (Utilities|Casting|CastToBuildingActor _array_element_2))
                      (Utilities|IsValid _asbuilding_actor
                        (:Is%20Valid
                          (if (Math|Boolean|ANDBoolean (not (Variables|/Script/Engine.Actor|GetbCanBeDamaged _asbuilding_actor)) (Utilities|Enum|Equal(Enum) (Variables|/Script/FortniteGame.BuildingActor|GetBuildingType _asbuilding_actor) (Variables|/Script/FortniteGame.BuildingActor|GetBuildingType _asbuilding_static_mesh)) (<= (Math|Vector|DistanceSquared(Vector) (Collision|GetActorBounds _asbuilding_actor false false) (Collision|GetActorBounds _asbuilding_static_mesh false false)) (Math|Float|Square 2.4e+02)))
                            (Gameplay|SilentDie _asbuilding_static_mesh false))))))
                  (:else))))))
        (:then_1
          (CallFunction|AdjustOverlappingPlayerPawnLocations)
          (CallFunction|SendDeployQuestEvent))))))

(fn SendDeployQuestEvent
  (bind _returnvalue (Actor|GetOwner))
  ;[0 0 0] Send a quest event for the Port-A-Bunker being deployed
  (Networking|SwitchHasAuthority
    (:Authority
      ;; Check if it was in water when deployed for additional context to send to the verb
      (LocalVariables|SetbOverlappingWater false)
      (bind _array_element (Utilities|Array|ForEachLoopwithBreak (Variables|Default|GetSpawnedBuildingActors))
        (:LoopBody
          (Utilities|IsValid _array_element
            (:Is%20Valid
              (bind _asbuilding_static_mesh (Utilities|Casting|CastToBuildingSMActor _array_element))
              (bind _staticmeshcomponent (Variables|/Script/FortniteGame.BuildingSMActor|GetStaticMeshComponent _asbuilding_static_mesh)
                (:then
                  (bind (_outactors _returnvalue_2) (Collision|ComponentOverlapActors _staticmeshcomponent (Transformation|GetWorldTransform _staticmeshcomponent) (Utilities|Array|MakeArray "ObjectTypeQuery13") "/Script/FortniteGame.WaterVolume_Athena" (Variables|Default|GetSpawnedBuildingActors)))
                  (if _returnvalue_2
                    (LocalVariables|SetbOverlappingWater true)))
                (:else)))))
        (:Completed
          _))
      (Utilities|IsValid _returnvalue
        (:Is%20Valid
          (bind _asfort_player_pawn_athena (Utilities|Casting|CastToFortPlayerPawnAthena _returnvalue))
          (if (LocalVariables|GetbOverlappingWater)
            (bind _result (CustomVerb|AddTagsforSubject (CustomVerb|MakeCustomVerbMessage _asfort_player_pawn_athena "(TagName=\"CustomVerb.DeployPortableShack\")") "DeployedOnWater" (LocalVariables|GetOverlappingWaterTags)))
            (CustomVerb|BroadcastCustomVerbMessage _result)
            (else
              (CustomVerb|BroadcastCustomVerbMessage (CustomVerb|MakeCustomVerbMessage _asfort_player_pawn_athena "(TagName=\"CustomVerb.DeployPortableShack\")")))))))))

(fn ConstructionScript
  (|CallParentFunction))
